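# Physics asset for the AlpacaLean_Gel medium male body's jacket, simulating the suit's hanging panels around pelvis and thigh capsules.
# Front, side, side-back and back panels on each side are two-capsule chains with limited swing and locked twist, hung from small base bodies.
# Bases constrain to the pelvis with locked rotation, except the free-rotating back bases.
# Only left-side panels have cross-links at their tips: front-to-side and side-back-to-back.
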PhysicsAsset /BRCosmetics/Characters/Player/Male/Medium/Bodies/M_MED_AlpacaLean_Gel/Meshes/M_MED_AlpacaLean_Gel_Jacket_Physics
bodies {
 base_suitBack_L [{shapeType=Capsule, length=1, radius=1, shapeName=dyn_suitSideFront_02_l_capsule}]
 base_suitBack_R [{shapeType=Capsule, length=1, radius=1, shapeName=dyn_suitSideFront_02_l_capsule}]
 base_suitSideBack_l [{shapeType=Capsule, length=0.505, radius=0.505, shapeName=base_suitSideBack_l_capsule, rotation=<rot 0 -90 0>}]
 base_suitSideBack_r [{shapeType=Capsule, length=0.505, radius=0.505, shapeName=base_suitSideBack_r_capsule, rotation=<rot 0 -90 0>}]
 base_suitSideFront_l [{shapeType=Capsule, length=0.505, radius=0.505, shapeName=base_suitSideFront_l_capsule, rotation=<rot 0 -90 0>}]
 base_suitSideFront_r [{shapeType=Capsule, length=0.505, radius=0.505, shapeName=base_suitSideFront_r_capsule, rotation=<rot 0 -90 0>}]
 base_suitSide_l [{shapeType=Capsule, length=0.505, radius=0.505, shapeName=base_suitSide_l_capsule, rotation=<rot 0 -90 0>}]
 base_suitSide_r [{shapeType=Capsule, length=0.505, radius=0.505, shapeName=base_suitSide_r_capsule, rotation=<rot 0 -90 0>}]
 dyn_suitBack_L_01 [{
   shapeType=Capsule
   length=7
   radius=6
   shapeName=dyn_suitSideFront_02_l_capsule
   center=<5.7391 -4.6666 -0.1677>
   rotation=<rot 1.7588 -90.0005 -0.0003>
  }]
 dyn_suitBack_L_02 [{
   shapeType=Capsule
   length=7
   radius=6
   shapeName=dyn_suitSideFront_02_l_capsule
   center=<5.7391 -4.6313 -0.5968>
   rotation=<rot -3.5266 -90.0005 -0.0003>
  }]
 dyn_suitBack_R_01 [{
   shapeType=Capsule
   length=7
   radius=6
   shapeName=dyn_suitSideFront_02_l_capsule
   center=<5.7391 -4.6649 -0.2098>
   rotation=<rot 1.2991 -90.0005 -0.0003>
  }]
 dyn_suitBack_R_02 [{
   shapeType=Capsule
   length=7
   radius=6
   shapeName=dyn_suitSideFront_02_l_capsule
   center=<5.7391 -4.6313 -0.5968>
   rotation=<rot -3.4685 -90.0005 -0.0003>
  }]
 dyn_suitSideBack_01_l [{
   shapeType=Capsule
   length=7
   radius=6
   shapeName=dyn_suitSideFront_02_l_capsule
   center=<5.7391 -4.6313 -0.5968>
   rotation=<rot -3.4685 -90.0005 -0.0003>
  }]
 dyn_suitSideBack_01_r [{
   shapeType=Capsule
   length=7
   radius=6
   shapeName=dyn_suitSideFront_02_l_capsule
   center=<5.7391 -4.6313 -0.5968>
   rotation=<rot -3.4685 -90.0005 -0.0003>
  }]
 dyn_suitSideBack_02_l [{
   shapeType=Capsule
   length=7
   radius=6
   shapeName=dyn_suitSideFront_02_l_capsule
   center=<5.7391 -4.6313 -0.5968>
   rotation=<rot -3.4685 -90.0005 -0.0003>
  }]
 dyn_suitSideBack_02_r [{
   shapeType=Capsule
   length=7
   radius=6
   shapeName=dyn_suitSideFront_02_l_capsule
   center=<5.7391 -4.6313 -0.5968>
   rotation=<rot -3.4685 -90.0005 -0.0003>
  }]
 dyn_suitSideFront_01_l [{
   shapeType=Capsule
   length=7
   radius=6
   shapeName=dyn_suitSideFront_02_l_capsule
   center=<5.739 -5.702 0.5693>
   rotation=<rot 5.7016 -90.0009 -0.0005>
  }]
 dyn_suitSideFront_01_r [{
   shapeType=Capsule
   length=7
   radius=6
   shapeName=dyn_suitSideFront_02_l_capsule
   center=<5.739 -6.7004 -0.019>
   rotation=<rot -0.1629 -90.0008 -0.0001>
  }]
 dyn_suitSideFront_02_l [{
   shapeType=Capsule
   length=7
   radius=6
   shapeName=dyn_suitSideFront_02_l_capsule
   center=<5.735 -5.422 1.1253>
   rotation=<rot 10.7552 -90.0349 -0.9424>
  }]
 dyn_suitSideFront_02_r [{
   shapeType=Capsule
   length=7
   radius=6
   shapeName=dyn_suitSideFront_02_l_capsule
   center=<5.7581 -5.6672 -0.2613>
   rotation=<rot -3.565 -89.8009 -0.9272>
  }]
 dyn_suitSide_01_l [{
   shapeType=Capsule
   length=7
   radius=6
   shapeName=dyn_suitSideFront_02_l_capsule
   center=<5.739 -7.8221 -1.008>
   rotation=<rot -7.3435 -90.0005 -0.0003>
  }]
 dyn_suitSide_01_r [{shapeType=Capsule, length=7, radius=6, shapeName=dyn_suitSideFront_02_l_capsule, center=<5.7391 -5.2417 0>, rotation=<rot 0 -90 0>}]
 dyn_suitSide_02_l [{
   shapeType=Capsule
   length=7
   radius=6
   shapeName=dyn_suitSideFront_02_l_capsule
   center=<5.764 -5.8804 -0.595>
   rotation=<rot -6.6548 -89.7504 -0.9321>
  }]
 dyn_suitSide_02_r [{
   shapeType=Capsule
   length=7
   radius=6
   shapeName=dyn_suitSideFront_02_l_capsule
   center=<5.7546 -5.2234 -0.1165>
   rotation=<rot -2.2874 -89.8215 -0.9265>
  }]
 pelvis [{shapeType=Capsule, length=8.774, radius=9.9763, center=<3.3821 -0.5468 0.2718>, rotation=<rot 0 90 180>}]
 thigh_l [
  {shapeType=Capsule, length=31.9507, radius=5.3, center=<-16.8538 -2.2098 0.0445>, rotation=<rot 0.2482 -92.6907 -89.0955>}
  {shapeType=Capsule, length=32.9743, radius=6.2045, center=<-17.0918 2.1916 -0.2425>, rotation=<rot 0.1349 -85.5737 -89.0716>}
 ]
 thigh_r [
  {shapeType=Capsule, length=31.9507, radius=5.3, center=<16.8882 2.9423 -0.0477>, rotation=<rot 0.2482 -92.6907 90.9045>}
  {shapeType=Capsule, length=32.0849, radius=6.4579, center=<16.9073 -1.3682 -0.0961>, rotation=<rot 0.1718 -87.9117 90.922>}
 ]
}
constraints [
 {bone1Name=thigh_l, bone2Name=pelvis, swing1LimitDegrees=45, swing2LimitDegrees=45, twistLimitDegrees=45}
 {bone1Name=thigh_r, bone2Name=pelvis, swing1LimitDegrees=45, swing2LimitDegrees=45, twistLimitDegrees=45}
 {
  bone1Name=dyn_suitSideFront_01_l
  bone2Name=base_suitSideFront_l
  swing1LimitDegrees=30
  swing2LimitDegrees=8
  twistLimitDegrees=45
  twistMotion=Locked
 }
 {
  bone1Name=dyn_suitSideFront_02_l
  bone2Name=dyn_suitSideFront_01_l
  swing1LimitDegrees=30
  swing2LimitDegrees=8
  twistLimitDegrees=45
  twistMotion=Locked
 }
 {
  bone1Name=dyn_suitSideFront_01_r
  bone2Name=base_suitSideFront_r
  swing1LimitDegrees=30
  swing2LimitDegrees=8
  twistLimitDegrees=45
  twistMotion=Locked
 }
 {
  bone1Name=dyn_suitSideFront_02_r
  bone2Name=dyn_suitSideFront_01_r
  swing1LimitDegrees=30
  swing2LimitDegrees=8
  twistLimitDegrees=45
  twistMotion=Locked
 }
 {
  bone1Name=dyn_suitSide_01_l
  bone2Name=base_suitSide_l
  swing1LimitDegrees=20
  swing2LimitDegrees=20
  twistLimitDegrees=45
  twistMotion=Locked
 }
 {
  bone1Name=dyn_suitSide_02_l
  bone2Name=dyn_suitSide_01_l
  swing1LimitDegrees=20
  swing2LimitDegrees=20
  twistLimitDegrees=45
  twistMotion=Locked
 }
 {
  bone1Name=dyn_suitSide_01_r
  bone2Name=base_suitSide_r
  swing1LimitDegrees=30
  swing2LimitDegrees=10
  twistLimitDegrees=45
  twistMotion=Locked
 }
 {
  bone1Name=dyn_suitSide_02_r
  bone2Name=dyn_suitSide_01_r
  swing1LimitDegrees=30
  swing2LimitDegrees=10
  twistLimitDegrees=45
  twistMotion=Locked
 }
 {
  bone1Name=base_suitSideBack_l
  bone2Name=pelvis
  swing1LimitDegrees=45
  swing1Motion=Locked
  swing2LimitDegrees=45
  swing2Motion=Locked
  twistLimitDegrees=45
  twistMotion=Locked
 }
 {
  bone1Name=base_suitSideBack_r
  bone2Name=pelvis
  swing1LimitDegrees=45
  swing1Motion=Locked
  swing2LimitDegrees=45
  swing2Motion=Locked
  twistLimitDegrees=45
  twistMotion=Locked
 }
 {
  bone1Name=base_suitSideFront_l
  bone2Name=pelvis
  swing1LimitDegrees=45
  swing1Motion=Locked
  swing2LimitDegrees=45
  swing2Motion=Locked
  twistLimitDegrees=45
  twistMotion=Locked
 }
 {
  bone1Name=base_suitSideFront_r
  bone2Name=pelvis
  swing1LimitDegrees=45
  swing1Motion=Locked
  swing2LimitDegrees=45
  swing2Motion=Locked
  twistLimitDegrees=45
  twistMotion=Locked
 }
 {
  bone1Name=base_suitSide_l
  bone2Name=pelvis
  swing1LimitDegrees=45
  swing1Motion=Locked
  swing2LimitDegrees=45
  swing2Motion=Locked
  twistLimitDegrees=45
  twistMotion=Locked
 }
 {
  bone1Name=base_suitSide_r
  bone2Name=pelvis
  swing1LimitDegrees=45
  swing1Motion=Locked
  swing2LimitDegrees=45
  swing2Motion=Locked
  twistLimitDegrees=45
  twistMotion=Locked
 }
 {
  bone1Name=base_suitBack_L
  bone2Name=pelvis
  swing1LimitDegrees=45
  swing1Motion=Free
  swing2LimitDegrees=45
  swing2Motion=Free
  twistLimitDegrees=45
  twistMotion=Free
 }
 {
  bone1Name=dyn_suitBack_L_01
  bone2Name=base_suitBack_L
  swing1LimitDegrees=20
  swing2LimitDegrees=10
  twistLimitDegrees=1
  twistMotion=Locked
 }
 {
  bone1Name=dyn_suitBack_L_02
  bone2Name=dyn_suitBack_L_01
  swing1LimitDegrees=20
  swing2LimitDegrees=10
  twistLimitDegrees=1
  twistMotion=Locked
 }
 {
  bone1Name=base_suitBack_R
  bone2Name=pelvis
  swing1LimitDegrees=45
  swing1Motion=Free
  swing2LimitDegrees=45
  swing2Motion=Free
  twistLimitDegrees=45
  twistMotion=Free
 }
 {
  bone1Name=dyn_suitBack_R_01
  bone2Name=base_suitBack_R
  swing1LimitDegrees=20
  swing2LimitDegrees=10
  twistLimitDegrees=1
  twistMotion=Locked
 }
 {
  bone1Name=dyn_suitBack_R_02
  bone2Name=dyn_suitBack_R_01
  swing1LimitDegrees=20
  swing2LimitDegrees=10
  twistLimitDegrees=1
  twistMotion=Locked
 }
 {
  bone1Name=dyn_suitSideFront_02_l
  bone2Name=dyn_suitSide_02_l
  swing1LimitDegrees=10
  swing2LimitDegrees=10
  swing2Motion=Locked
  twistLimitDegrees=20
  twistMotion=Free
 }
 {
  bone1Name=dyn_suitSideBack_02_l
  bone2Name=dyn_suitSideBack_01_l
  swing1LimitDegrees=20
  swing2LimitDegrees=10
  twistLimitDegrees=45
  twistMotion=Locked
 }
 {
  bone1Name=dyn_suitSideBack_01_l
  bone2Name=base_suitSideBack_l
  swing1LimitDegrees=20
  swing2LimitDegrees=10
  twistLimitDegrees=45
  twistMotion=Locked
 }
 {
  bone1Name=dyn_suitSideBack_01_r
  bone2Name=base_suitSideBack_r
  swing1LimitDegrees=20
  swing2LimitDegrees=10
  twistLimitDegrees=45
  twistMotion=Locked
 }
 {
  bone1Name=dyn_suitSideBack_02_r
  bone2Name=dyn_suitSideBack_01_r
  swing1LimitDegrees=20
  swing2LimitDegrees=10
  twistLimitDegrees=45
  twistMotion=Locked
 }
 {
  bone1Name=dyn_suitSideBack_02_l
  bone2Name=dyn_suitBack_L_02
  swing1LimitDegrees=45
  swing1Motion=Free
  swing2LimitDegrees=45
  swing2Motion=Free
  twistLimitDegrees=45
  twistMotion=Free
 }
]
