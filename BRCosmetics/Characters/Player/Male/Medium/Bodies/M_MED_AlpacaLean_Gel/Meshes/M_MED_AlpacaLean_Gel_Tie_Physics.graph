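# Physics asset for the tie on the M_MED_AlpacaLean_Gel player character, set up for secondary motion of the tie bones.
# Five capsule bodies form a dyn_tie chain from base_tie, which is angularly locked to spine_05.
# Chain joints lock twist and limit swing, with the widest range at the top link.
# Pelvis, spine and forearm capsules, joined by angularly free constraints, provide body shapes for the tie to collide against.
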PhysicsAsset /BRCosmetics/Characters/Player/Male/Medium/Bodies/M_MED_AlpacaLean_Gel/Meshes/M_MED_AlpacaLean_Gel_Tie_Physics
bodies {
 base_tie [{shapeType=Capsule, length=0.505, radius=0.505, shapeName=base_tie_capsule, rotation=<rot 0 -90 0>}]
 dyn_tie_01 [{
   shapeType=Capsule
   length=4.9552
   radius=2.1751
   shapeName=dyn_tie_01_capsule
   center=<4.4025 0 -1.6317>
   rotation=<rot -1.9292 -91.6136 -93.0973>
  }]
 dyn_tie_02 [{
   shapeType=Capsule
   length=4.9552
   radius=2.1751
   shapeName=dyn_tie_01_capsule
   center=<4.4016 -0.0005 -1.6161>
   rotation=<rot -1.9292 -91.6136 -93.0973>
  }]
 dyn_tie_03 [{
   shapeType=Capsule
   length=4.9552
   radius=2.1751
   shapeName=dyn_tie_01_capsule
   center=<4.3759 -0.0155 -1.1501>
   rotation=<rot -1.9292 -91.6136 -93.0973>
  }]
 dyn_tie_04 [{
   shapeType=Capsule
   length=4.9552
   radius=2.1751
   shapeName=dyn_tie_01_capsule
   center=<4.3746 -0.0225 -0.9411>
   rotation=<rot -1.9292 -91.6136 -92.2567>
  }]
 dyn_tie_05 [{
   shapeType=Capsule
   length=2.2223
   radius=2.1751
   shapeName=dyn_tie_01_capsule
   center=<3.4022 -0.0403 -0.0538>
   rotation=<rot -1.9292 -91.6136 -93.0973>
  }]
 lowerarm_l [{
   shapeType=Capsule
   length=19.5622
   radius=5.0364
   shapeName=lowerarm_l_capsule
   center=<9.3057 0 -0.4554>
   rotation=<rot 86.4313 179.4209 -179.3527>
  }]
 lowerarm_r [{
   shapeType=Capsule
   length=19.5622
   radius=4.9886
   shapeName=lowerarm_r_capsule
   center=<-9.3139 0 0.7037>
   rotation=<rot 86.651 179.4659 0.692>
  }]
 pelvis [{shapeType=Capsule, length=9.8585, radius=12.5662, center=<-1.6033 -1.1214 -0.2602>, rotation=<rot 4.4575 -90 2.0111>}]
 spine_01 [{shapeType=Capsule, length=10.4491, radius=11.7099, shapeName=spine_01_capsule, center=<0 0.715 0>, rotation=<rot 0 -90 0>}]
 spine_02 [{shapeType=Capsule, length=15, radius=5.0716, shapeName=spine_02_capsule, center=<0.0439 6.8134 6.3976>, rotation=<rot -88.6801 -0.0019 0>}]
 spine_03 [
  {shapeType=Capsule, length=8.7035, radius=12.1, shapeName=spine_01_capsule, center=<0 -0.0213 0>, rotation=<rot 0 -90 0>}
  {
   shapeType=Capsule
   length=15.5262
   radius=4.4624
   shapeName=spine_03_capsule
   center=<0.2192 7.6257 6.8923>
   rotation=<rot -87.4655 -5.0457 9.7838>
  }
 ]
 spine_05 [{shapeType=Capsule, length=13.7767, radius=12.6928, shapeName=spine_05_capsule, center=<-5.4239 -0.2178 0>}]
}
constraints [
 {bone1Name=dyn_tie_01, bone2Name=base_tie, swing1LimitDegrees=20, swing2LimitDegrees=50, twistLimitDegrees=45, twistMotion=Locked}
 {bone1Name=dyn_tie_02, bone2Name=dyn_tie_01, swing1LimitDegrees=12, swing2LimitDegrees=10, twistLimitDegrees=45, twistMotion=Locked}
 {bone1Name=dyn_tie_03, bone2Name=dyn_tie_02, swing1LimitDegrees=12, swing2LimitDegrees=10, twistLimitDegrees=45, twistMotion=Locked}
 {bone1Name=dyn_tie_04, bone2Name=dyn_tie_03, swing1LimitDegrees=12, swing2LimitDegrees=10, twistLimitDegrees=45, twistMotion=Locked}
 {bone1Name=dyn_tie_05, bone2Name=dyn_tie_04, swing1LimitDegrees=12, swing2LimitDegrees=10, twistLimitDegrees=45, twistMotion=Locked}
 {
  bone1Name=base_tie
  bone2Name=spine_05
  swing1LimitDegrees=45
  swing1Motion=Locked
  swing2LimitDegrees=45
  swing2Motion=Locked
  twistLimitDegrees=45
  twistMotion=Locked
 }
 {
  bone1Name=spine_03
  bone2Name=spine_02
  swing1LimitDegrees=45
  swing1Motion=Free
  swing2LimitDegrees=45
  swing2Motion=Free
  twistLimitDegrees=45
  twistMotion=Free
 }
 {
  bone1Name=spine_02
  bone2Name=spine_01
  swing1LimitDegrees=45
  swing1Motion=Free
  swing2LimitDegrees=45
  swing2Motion=Free
  twistLimitDegrees=45
  twistMotion=Free
 }
 {
  bone1Name=spine_01
  bone2Name=pelvis
  swing1LimitDegrees=45
  swing1Motion=Free
  swing2LimitDegrees=45
  swing2Motion=Free
  twistLimitDegrees=45
  twistMotion=Free
 }
 {
  bone1Name=spine_05
  bone2Name=spine_03
  swing1LimitDegrees=45
  swing1Motion=Free
  swing2LimitDegrees=45
  swing2Motion=Free
  twistLimitDegrees=45
  twistMotion=Free
 }
 {
  bone1Name=lowerarm_r
  bone2Name=spine_05
  swing1LimitDegrees=45
  swing1Motion=Free
  swing2LimitDegrees=45
  swing2Motion=Free
  twistLimitDegrees=45
  twistMotion=Free
 }
 {
  bone1Name=lowerarm_l
  bone2Name=spine_05
  swing1LimitDegrees=45
  swing1Motion=Free
  swing2LimitDegrees=45
  swing2Motion=Free
  twistLimitDegrees=45
  twistMotion=Free
 }
]
