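# Post-process style cosmetic AnimBP for the AlpacaLean_Gel male body, adding secondary motion to its suit jacket, tie and laces.
# The attached parent's pose feeds a state-aware Control Rig, then jacket RigidBody physics.
# A front-end-idle flag selects between two tie simulations, followed by additive bone offsets and Anim Dynamics laces.
# Event graph updates that flag, gravity override and an LOD-above-1 flag that enables pelvis_alt scaling.
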AnimBlueprint /BRCosmetics/Characters/Player/Male/Medium/Bodies/M_MED_AlpacaLean_Gel/Meshes/M_MED_AlpacaLean_Gel_AnimBP
target_skeleton=&/M_MED_AlpacaLean_Gel_Skeleton
graphs [
 AnimGraph {nodes [
   n1 = "Transform (Modify) Bone - Bone: base_suitSide_r"!(ComponentPose, bAlphaBoolEnabled:bool=True, Translation="0,0,0", Rotation="-9.299,-4.376,-0.007", Scale="1,1,1") -> (Pose) {
    alphaInputType=Bool
    rotationMode=BMM_Additive
    rotationSpace=BCS_BoneSpace
    scaleSpace=BCS_BoneSpace
    translationMode=BMM_Additive
    translationSpace=BCS_BoneSpace
    alphaBoolBlend {blendInTime=0.6, blendOutTime=1}
    boneToModify {boneName=base_suitSide_r}
    rotation=<rot -9.299 -4.376 -0.007>
   }
   n2 = "Transform (Modify) Bone - Bone: dyn_suitSideFront_02_r"!(ComponentPose=@n1, bAlphaBoolEnabled:bool=True, Translation="0,0,0", Rotation="4.978,16,0.7", Scale="1,1,1") -> (Pose) {
    alphaInputType=Bool
    rotationMode=BMM_Additive
    rotationSpace=BCS_BoneSpace
    scaleSpace=BCS_BoneSpace
    translationMode=BMM_Additive
    translationSpace=BCS_BoneSpace
    alphaBoolBlend {blendInTime=0.6, blendOutTime=1}
    boneToModify {boneName=dyn_suitSideFront_02_r}
    rotation=<rot 4.978 16 0.7>
   }
   "Transform (Modify) Bone - Bone: base_suitSideFront_r"!(ComponentPose=@n2, bAlphaBoolEnabled:bool=True, Translation="0,0,0", Rotation="0.974,-15,-0.059", Scale="1,1,1") -> (Pose) {
    alphaInputType=Bool
    rotationMode=BMM_Additive
    rotationSpace=BCS_BoneSpace
    scaleSpace=BCS_BoneSpace
    translationMode=BMM_Additive
    translationSpace=BCS_BoneSpace
    alphaBoolBlend {blendInTime=0.2, blendOutTime=1}
    boneToModify {boneName=base_suitSideFront_r}
    rotation=<rot 0.974 -15 -0.059>
   }
   n3 = RigidBody!(ComponentPose=Local To Component(LocalPose=Use cached pose 'jacket_pysics'()), Alpha, OverrideWorldGravity:Vector="0,0,0", ExternalForce) -> (Pose) {
    bOverrideWorldGravity=true
    overridePhysicsAsset=&/M_MED_AlpacaLean_Gel_Tie_Physics
    simulationSpace=BaseBoneSpace
    baseBoneRef {boneName=spine_04}
    componentAppliedLinearAccClamp=<1000 1000 1000>
    simSpaceSettings {
     dampingAlpha=0.5
     maxAngularAcceleration=20
     maxAngularVelocity=5
     maxLinearAcceleration=1000
     maxLinearVelocity=300
     velocityScaleZ=0.25
     worldAlpha=1
     externalLinearDragV=<0.3 0.3 0.3>
    }
   }
   n4 = Component To Local(ComponentPose=@n3)
   n5 = RigidBody(ComponentPose=Local To Component(LocalPose=Use cached pose 'jacket_pysics'()), OverrideWorldGravity="0,0,0") {
    bOverrideWorldGravity=true
    overridePhysicsAsset=&/M_MED_AlpacaLean_Gel_Tie_FE_Physics
    simulationSpace=BaseBoneSpace
    baseBoneRef {boneName=spine_04}
    componentAppliedLinearAccClamp=<1000 1000 1000>
    simSpaceSettings {
     dampingAlpha=0.5
     maxAngularAcceleration=20
     maxAngularVelocity=5
     maxLinearAcceleration=1000
     maxLinearVelocity=300
     velocityScaleZ=0.25
     worldAlpha=1
     externalLinearDragV=<0.3 0.3 0.3>
    }
   }
   n6 = Component To Local(ComponentPose=@n5)
   n7 = Get FE_Idle() -> (FE_Idle:bool)
   n8 = "Transform (Modify) Bone - Bone: dyn_tie_02"!(
    ComponentPose=Local To Component(LocalPose=Blend Poses by bool(BlendPose_0=@n6, BlendPose_1=@n4, bActiveValue=@n7) {blendType=Linear})
    bAlphaBoolEnabled:bool=True
    Translation="0,0,0"
    Rotation="0.229,-0.027,-13.027"
    Scale="1,1,1"
   ) -> (Pose) {
    alphaInputType=Bool
    rotationMode=BMM_Additive
    translationMode=BMM_Additive
    alphaBoolBlend {blendInTime=1, blendOutTime=1}
    boneToModify {boneName=dyn_tie_02}
    rotation=<rot 0.229 -0.027 -13.027>
   }
   n9 = "Transform (Modify) Bone - Bone: dyn_tie_03"!(ComponentPose=@n8, bAlphaBoolEnabled:bool=True, Translation="0,0,0", Rotation="0.229,-0.019,-12.235", Scale="1,1,1") -> (Pose) {
    alphaInputType=Bool
    rotationMode=BMM_Additive
    translationMode=BMM_Additive
    alphaBoolBlend {blendInTime=1, blendOutTime=1}
    boneToModify {boneName=dyn_tie_03}
    rotation=<rot 0.229 -0.019 -12.235>
   }
   n10 = "Transform (Modify) Bone - Bone: dyn_suitSideFront_01_r"!(ComponentPose=@n9, bAlphaBoolEnabled:bool=True, Translation="0,0,0", Rotation="0.017,-4.922,0.032", Scale="1,1,1") -> (Pose) {
    alphaInputType=Bool
    rotationMode=BMM_Additive
    rotationSpace=BCS_BoneSpace
    scaleSpace=BCS_BoneSpace
    translationMode=BMM_Additive
    translationSpace=BCS_BoneSpace
    alphaBoolBlend {blendInTime=0.2, blendOutTime=0.2}
    boneToModify {boneName=dyn_suitSideFront_01_r}
    rotation=<rot 0.017 -4.922 0.032>
   }
   n11 = "Transform (Modify) Bone - Bone: dyn_suitSideFront_01_l"!(ComponentPose=@n10, bAlphaBoolEnabled:bool=True, Translation="0,0,0", Rotation="0.010,-4.817,-2.829", Scale="1,1,1") -> (Pose) {
    alphaInputType=Bool
    rotationMode=BMM_Additive
    rotationSpace=BCS_BoneSpace
    scaleSpace=BCS_BoneSpace
    translationMode=BMM_Additive
    translationSpace=BCS_BoneSpace
    alphaBoolBlend {blendInTime=1, blendOutTime=1}
    boneToModify {boneName=dyn_suitSideFront_01_l}
    rotation=<rot 0.01 -4.817 -2.829>
   }
   n12 = "Transform (Modify) Bone - Bone: dyn_suitSideFront_01_r"(ComponentPose=@n11, bAlphaBoolEnabled=True, Translation="0,0,0", Rotation="0.010,-4.817,-2.829", Scale="1,1,1") {
    alphaInputType=Bool
    rotationMode=BMM_Additive
    rotationSpace=BCS_BoneSpace
    scaleSpace=BCS_BoneSpace
    translationMode=BMM_Additive
    translationSpace=BCS_BoneSpace
    alphaBoolBlend {blendInTime=1, blendOutTime=1}
    boneToModify {boneName=dyn_suitSideFront_01_r}
    rotation=<rot 0.01 -4.817 -2.829>
   }
   n13 = Anim Dynamics - Bone: dyn_lace_L_01!(ComponentPose=@n12, Alpha=1, GravityScale:float=1, GravityOverride:Vector="0,0,-490", ExternalForce="0,0,0") -> (Pose) {
    angularBiasOverride=0.9999
    angularDampingOverride=1
    angularSpringConstant=5000
    bAngularSpring=true
    bGravityOverrideInSimSpace=true
    bOverrideAngularBias=true
    bOverrideAngularDamping=true
    bOverrideLinearDamping=true
    bUseGravityOverride=true
    bUsePlanarLimit=false
    linearDampingOverride=0.9999
    simulationSpace=RootRelative
    alphaBoolBlend {blendOption=Linear}
    boundBone {boneName=dyn_lace_L_01}
    componentAppliedLinearAccClamp=<5000 5000 5000>
    componentLinearAccScale=<0.5 0.5 0.5>
    componentLinearVelScale=<0.5 0.5 0.5>
    gravityOverride {z=-980}
    physicsBodyDefinitions [{
      sphereCollisionRadius=0
      boundBone {boneName=dyn_lace_L_01}
      boxExtents=<10 5 5>
      constraintSetup {angularLimitsMax=<7 5 5>, angularLimitsMin=<-10 -80 -20>, angularTarget=<1 0 0>}
      localJointOffset=<3 0 0>
     }]
    relativeSpaceBone {boneName=pelvis}
   }
   n14 = Anim Dynamics - Bone: dyn_lace_L_02!(ComponentPose=@n13, Alpha=1, GravityScale:float=1, GravityOverride:Vector="0,0,-490", ExternalForce="0,0,0") -> (Pose) {
    angularBiasOverride=0.9999
    angularDampingOverride=1
    angularSpringConstant=5000
    bAngularSpring=true
    bGravityOverrideInSimSpace=true
    bOverrideAngularBias=true
    bOverrideAngularDamping=true
    bOverrideLinearDamping=true
    bUseGravityOverride=true
    bUsePlanarLimit=false
    linearDampingOverride=0.9999
    simulationSpace=RootRelative
    alphaBoolBlend {blendOption=Linear}
    boundBone {boneName=dyn_lace_L_02}
    componentAppliedLinearAccClamp=<5000 5000 5000>
    componentLinearAccScale=<0.5 0.5 0.5>
    componentLinearVelScale=<0.5 0.5 0.5>
    gravityOverride {z=-980}
    physicsBodyDefinitions [{
      sphereCollisionRadius=0
      boundBone {boneName=dyn_lace_L_02}
      boxExtents=<10 5 5>
      constraintSetup {angularLimitsMax=<7 5 20>, angularLimitsMin=<-10 -80 0>, angularTarget=<1 0 0>}
      localJointOffset=<3 0 0>
     }]
    relativeSpaceBone {boneName=pelvis}
   }
   n15 = Anim Dynamics - Bone: dyn_lace_R_01!(ComponentPose=@n14, Alpha=1, GravityScale:float=1, GravityOverride:Vector="0,0,-980", ExternalForce="0,0,0") -> (Pose) {
    angularBiasOverride=0.9999
    angularDampingOverride=1
    angularSpringConstant=5000
    bAngularSpring=true
    bGravityOverrideInSimSpace=true
    bOverrideAngularBias=true
    bOverrideAngularDamping=true
    bOverrideLinearDamping=true
    bUseGravityOverride=true
    bUsePlanarLimit=false
    linearDampingOverride=0.9999
    simulationSpace=RootRelative
    alphaBoolBlend {blendOption=Linear}
    boundBone {boneName=dyn_lace_R_01}
    componentAppliedLinearAccClamp=<5000 5000 5000>
    componentLinearAccScale=<0.5 0.5 0.5>
    componentLinearVelScale=<0.5 0.5 0.5>
    gravityOverride {z=-980}
    physicsBodyDefinitions [{
      sphereCollisionRadius=0
      boundBone {boneName=dyn_lace_R_01}
      boxExtents=<10 5 5>
      constraintSetup {angularLimitsMax=<7 5 5>, angularLimitsMin=<-10 -80 -20>, angularTarget=<1 0 0>}
      localJointOffset=<3 0 0>
     }]
    relativeSpaceBone {boneName=foot_l}
   }
   n16 = Anim Dynamics - Bone: dyn_lace_R_02!(ComponentPose=@n15, Alpha=1, GravityScale:float=1, GravityOverride:Vector="0,0,-980", ExternalForce="0,0,0") -> (Pose) {
    angularBiasOverride=0.9999
    angularDampingOverride=1
    angularSpringConstant=5000
    bAngularSpring=true
    bGravityOverrideInSimSpace=true
    bOverrideAngularBias=true
    bOverrideAngularDamping=true
    bOverrideLinearDamping=true
    bUseGravityOverride=true
    bUsePlanarLimit=false
    linearDampingOverride=0.9999
    simulationSpace=RootRelative
    alphaBoolBlend {blendOption=Linear}
    boundBone {boneName=dyn_lace_R_02}
    componentAppliedLinearAccClamp=<5000 5000 5000>
    componentLinearAccScale=<0.6 0.6 0.6>
    componentLinearVelScale=<0.6 0.6 0.6>
    gravityOverride {z=-980}
    physicsBodyDefinitions [{
      sphereCollisionRadius=0
      boundBone {boneName=dyn_lace_R_02}
      boxExtents=<10 5 5>
      constraintSetup {angularLimitsMax=<7 5 20>, angularLimitsMin=<-10 -50 -5>, angularTarget=<1 0 0>}
      localJointOffset=<3 0 0>
     }]
    relativeSpaceBone {boneName=foot_l}
   }
   n17 = "Transform (Modify) Bone - Bone: pelvis_alt"!(ComponentPose=@n16, bAlphaBoolEnabled:bool=Get ALT_COR() -> (ALT_COR:bool), Translation="0,0,0", Rotation="0,0,0", Scale="0.5,0.5,0.5") -> (Pose) {alphaInputType=Bool, scaleMode=BMM_Replace, boneToModify {boneName=pelvis_alt}, scale=<0.5 0.5 0.5>}
   Output Pose(Result=Component To Local(ComponentPose=@n17)) {layerGroup=None}
   n18 = Control Rig!(
    Source=Copy Pose From Mesh() {bUseAttachedParent=true}
    Alpha
    isCrouching_CR:bool=Get bIsCrouching()
    isKneeSliding_CR:bool=Get bIsKneeSliding()
    isBaseballSliding_CR:bool=Get bIsBaseballSliding()
    isFE_CR:bool=Get bIsInFrontEnd()
    isDBNO_CR:bool=Get bIsDBNO()
   ) -> (Pose) {controlRigAssetReference {blueprintRigClass=&/M_MED_AlpacaLean_Gel_CtrlRig.M_MED_AlpacaLean_Gel_CtrlRig_C}}
   Save cached pose 'PreSim'(Pose=@n18) {cacheName=PreSim}
   n19 = RigidBody(ComponentPose=Local To Component(LocalPose=Use cached pose 'PreSim'()), OverrideWorldGravity="0,0,0") {
    bOverrideWorldGravity=true
    overridePhysicsAsset=&/M_MED_AlpacaLean_Gel_Jacket_Physics
    simulationSpace=BaseBoneSpace
    baseBoneRef {boneName=pelvis}
    componentAppliedLinearAccClamp=<3000 3000 3000>
    componentLinearAccScale=<0.2 0.2 0.2>
    componentLinearVelScale=<0.2 0.2 0.2>
    simSpaceSettings {
     dampingAlpha=0.1
     maxAngularAcceleration=30
     maxAngularVelocity=6
     maxLinearAcceleration=1000
     maxLinearVelocity=300
     velocityScaleZ=0.5
     worldAlpha=0.6
     externalLinearDragV=<0.3 0.3 1>
    }
   }
   Save cached pose 'jacket_pysics'(Pose=Component To Local(ComponentPose=@n19)) {cacheName=jacket_pysics}
  ]}
 EventGraph {nodes [
   Event BlueprintUpdateAnimation()
   > n1 = Sequence()
   > n2 = "Gravity Override State Function (Parameter Asset) (Nativized)"(
    CharacterPartAnimInstance=Self-Reference()
    GameTimeInSeconds=GetGameTimeInSeconds()
    DebugJointName=spine_05
    Parameters=&/M_MED_AlpacaLean_Gel_Parameters
   )
   > Set Gravity Override(Gravity Override=@n2)
   n3 = Get bIsPlayingEmote()
   Set FE_Idle(execute=@n1.then_1, FE_Idle:bool=AND Boolean(A=Get bIsInFrontEnd(), B=NOT Boolean(A=@n3))) -> (Output_Get:bool)
   > Set ALT_COR(ALT_COR:bool="integer > integer"(A=Get CurrentLOD(), B=1)) -> (Output_Get:bool)
  ]}
]
variables [
 ALT_COR {type=boolean}
 FE_Idle {type=boolean}
 Gravity Override {type=Vector}
]
