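# Physics asset for the AlpacaLean Gel male character's FaceAcc mesh, providing hair secondary motion with capsule bodies.
# Five three-bone side and back chains hang from base bones on the head, and two top chains attach directly, all with twist-locked, narrow swing limits.
# Angular-free cross-constraints tie neighboring side and back chains together; the back-left and back-right bases are also angular-free.
# Head, neck, upper spine and clavicle capsules form a rotation-locked body frame.
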PhysicsAsset /BRCosmetics/Characters/Player/Male/Medium/Bodies/M_MED_AlpacaLean_Gel/Meshes/Parts/M_MED_AlpacaLean_Gel_FaceAcc_Physics
bodies {
 base_hair_01_back [{shapeType=Capsule, length=1, radius=1, shapeName=base_hair_01_back_capsule, rotation=<rot 0 -90 0>}]
 base_hair_01_back_lf [{shapeType=Capsule, length=1, radius=1, shapeName=base_hair_01_back_capsule, rotation=<rot 0 -90 0>}]
 base_hair_01_back_rt [{shapeType=Capsule, length=1, radius=1, shapeName=base_hair_01_back_capsule, rotation=<rot 0 -90 0>}]
 base_hair_01_lf [{shapeType=Capsule, length=1, radius=1, shapeName=base_hair_01_back_capsule, rotation=<rot 0 -90 0>}]
 base_hair_01_rt [{shapeType=Capsule, length=1, radius=1, shapeName=base_hair_01_back_capsule, rotation=<rot 0 -90 0>}]
 clavicle_l [{
   shapeType=Capsule
   length=12.5048
   radius=5.3842
   shapeName=clavicle_l_capsule
   center=<10.8379 -3.8274 2.3501>
   rotation=<rot -0.0008 -89.9999 105.329>
  }]
 clavicle_r [{
   shapeType=Capsule
   length=12.5048
   radius=5.3842
   shapeName=clavicle_l_capsule
   center=<-11.0197 3.8274 -3.0133>
   rotation=<rot -0.0008 -89.9999 105.329>
  }]
 dyn_hair_01_back [{shapeType=Capsule, length=4.7379, radius=2.2621, shapeName=dyn_hair_01_back_capsule, center=<2 0 0>}]
 dyn_hair_01_back_lf [{
   shapeType=Capsule
   length=4.7379
   radius=2.2621
   shapeName=dyn_hair_01_back_lf_capsule
   center=<2.0001 0 0>
   rotation=<rot 0.0168 0.0173 -4.4215>
  }]
 dyn_hair_01_back_rt [{shapeType=Capsule, length=2.8632, radius=2.2621, shapeName=dyn_hair_01_back_rt_capsule, center=<-2 0 0>, rotation=<rot 0 0 -5.9948>}]
 dyn_hair_01_lf [{shapeType=Capsule, length=6.5, radius=2.2, shapeName=dyn_hair_01_back_capsule, center=<1.668 0 0>}]
 dyn_hair_01_rt [{shapeType=Capsule, length=7.5, radius=2, shapeName=dyn_hair_01_back_capsule, center=<-2.1474 0.024 -0.032>}]
 dyn_hair_02_back [{shapeType=Capsule, length=1.8524, radius=3.45, shapeName=dyn_hair_03_back_capsule, center=<1.8884 -2.077 0>}]
 dyn_hair_02_back_lf [{
   shapeType=Capsule
   length=4.7328
   radius=3.45
   shapeName=dyn_hair_03_back_lf_capsule
   center=<1.3631 -3.1171 -0.178>
   rotation=<rot 0.0168 0.0173 -4.4222>
  }]
 dyn_hair_02_back_rt [{
   shapeType=Capsule
   length=4.7328
   radius=3.45
   shapeName=dyn_hair_03_back_lf_capsule
   center=<-1.0649 2.1381 0.2236>
   rotation=<rot 7.3561 -0.5533 -4.459>
  }]
 dyn_hair_02_lf [{shapeType=Capsule, length=6.5, radius=3.2, shapeName=dyn_hair_01_back_capsule, center=<1.668 -1.5652 0>}]
 dyn_hair_02_rt [{shapeType=Capsule, length=3.5, radius=3.5, shapeName=dyn_hair_01_back_capsule, center=<-2.1464 0.6629 -0.0274>}]
 dyn_hair_03_back [{shapeType=Capsule, length=2.3842, radius=3.45, shapeName=dyn_hair_03_back_capsule, center=<1.9732 -1.6634 0>, rotation=<rot 0 -17.9393 0>}]
 dyn_hair_03_back_lf [{
   shapeType=Capsule
   length=4.7328
   radius=3.45
   shapeName=dyn_hair_03_back_lf_capsule
   center=<1.3631 -3.1171 -0.178>
   rotation=<rot 0.0168 0.0173 -4.4222>
  }]
 dyn_hair_03_back_rt [{
   shapeType=Capsule
   length=4.7328
   radius=3.45
   shapeName=dyn_hair_03_back_lf_capsule
   center=<-1.1811 2.5852 0.2431>
   rotation=<rot 7.3561 -0.5533 -4.459>
  }]
 dyn_hair_03_lf [{shapeType=Capsule, length=4.5, radius=3.5, shapeName=dyn_hair_03_back_capsule, center=<1.668 -1.5652 0>}]
 dyn_hair_03_rt [{shapeType=Capsule, length=1, radius=3.5, shapeName=dyn_hair_03_back_capsule, center=<-1.0421 1.5779 0.3018>, rotation=<rot 3.939 0 0>}]
 dyn_hair_top_L_01 [{
   shapeType=Capsule
   length=2.8033
   radius=0.8462
   shapeName=dyn_hair_top_L_01_capsule
   center=<4.203 -0.1111 0.4503>
   rotation=<rot 0 -89.3417 -81.9822>
  }]
 dyn_hair_top_L_02 [{
   shapeType=Capsule
   length=2.973
   radius=1.3462
   shapeName=dyn_hair_top_L_02_capsule
   center=<3.5384 -0.0323 1.854>
   rotation=<rot 0 -88.9012 -88.1346>
  }]
 dyn_hair_top_L_03 [{
   shapeType=Capsule
   length=2.2366
   radius=0.8462
   shapeName=dyn_hair_top_L_03_capsule
   center=<4.3156 0.0534 1.7569>
   rotation=<rot 0 -88.6085 -68.7357>
  }]
 dyn_hair_top_R_01 [{
   shapeType=Capsule
   length=1.6129
   radius=0.505
   shapeName=dyn_hair_top_R_01_capsule
   center=<1.9841 -0.2352 -1.0178>
   rotation=<rot 0 83.2388 -88.95>
  }]
 dyn_hair_top_R_02 [{
   shapeType=Capsule
   length=3.7703
   radius=0.7041
   shapeName=dyn_hair_top_R_02_capsule
   center=<1.0927 0.0212 -1.5727>
   rotation=<rot 0 -88.8902 -88.2174>
  }]
 dyn_hair_top_R_03 [{
   shapeType=Capsule
   length=1.9025
   radius=0.505
   shapeName=dyn_hair_top_R_03_capsule
   center=<1.1955 0.2925 -1.3746>
   rotation=<rot 0 103.7471 -81.6998>
  }]
 head [{shapeType=Capsule, length=0, radius=8.9924, center=<3.2125 2.9665 -0.1299>, rotation=<rot 0 0 -90>}]
 neck_02 [{
   shapeType=Capsule
   length=4.9284
   radius=5.8613
   shapeName=neck_01_capsule
   center=<-0.0151 1.0689 0>
   rotation=<rot 0.0116 -86.0105 -90.1656>
  }]
 spine_05 [
  {shapeType=Capsule, length=19.1395, radius=10.3203, shapeName=spine_05_capsule, center=<3.6262 -1.3046 -1.603>, rotation=<rot 0 -81.2455 0>}
  {shapeType=Capsule, length=10.1853, radius=7.3963, shapeName=spine_05_capsule, center=<7.6655 -1.0289 0>, rotation=<rot 0 15.2944 0>}
 ]
}
constraints [
 {
  bone1Name=base_hair_01_back
  bone2Name=head
  swing1LimitDegrees=45
  swing1Motion=Locked
  swing2LimitDegrees=45
  swing2Motion=Locked
  twistLimitDegrees=45
  twistMotion=Locked
 }
 {
  bone1Name=dyn_hair_01_back
  bone2Name=base_hair_01_back
  swing1LimitDegrees=20
  swing2LimitDegrees=20
  twistLimitDegrees=45
  twistMotion=Locked
 }
 {
  bone1Name=dyn_hair_02_back
  bone2Name=dyn_hair_01_back
  swing1LimitDegrees=20
  swing2LimitDegrees=20
  twistLimitDegrees=45
  twistMotion=Locked
 }
 {
  bone1Name=dyn_hair_03_back
  bone2Name=dyn_hair_02_back
  swing1LimitDegrees=20
  swing2LimitDegrees=20
  twistLimitDegrees=45
  twistMotion=Locked
 }
 {
  bone1Name=head
  bone2Name=neck_02
  swing1LimitDegrees=45
  swing1Motion=Locked
  swing2LimitDegrees=45
  swing2Motion=Locked
  twistLimitDegrees=45
  twistMotion=Locked
 }
 {
  bone1Name=neck_02
  bone2Name=spine_05
  swing1LimitDegrees=45
  swing1Motion=Locked
  swing2LimitDegrees=45
  swing2Motion=Locked
  twistLimitDegrees=45
  twistMotion=Locked
 }
 {
  bone1Name=clavicle_l
  bone2Name=spine_05
  swing1LimitDegrees=45
  swing1Motion=Locked
  swing2LimitDegrees=45
  swing2Motion=Locked
  twistLimitDegrees=45
  twistMotion=Locked
 }
 {
  bone1Name=clavicle_r
  bone2Name=spine_05
  swing1LimitDegrees=45
  swing1Motion=Locked
  swing2LimitDegrees=45
  swing2Motion=Locked
  twistLimitDegrees=45
  twistMotion=Locked
 }
 {
  bone1Name=base_hair_01_lf
  bone2Name=head
  swing1LimitDegrees=45
  swing1Motion=Locked
  swing2LimitDegrees=45
  swing2Motion=Locked
  twistLimitDegrees=45
  twistMotion=Locked
 }
 {bone1Name=dyn_hair_01_lf, bone2Name=base_hair_01_lf, swing1LimitDegrees=10, swing2LimitDegrees=20, twistLimitDegrees=45, twistMotion=Locked}
 {bone1Name=dyn_hair_02_lf, bone2Name=dyn_hair_01_lf, swing1LimitDegrees=10, swing2LimitDegrees=20, twistLimitDegrees=45, twistMotion=Locked}
 {bone1Name=dyn_hair_03_lf, bone2Name=dyn_hair_02_lf, swing1LimitDegrees=10, swing2LimitDegrees=20, twistLimitDegrees=45, twistMotion=Locked}
 {
  bone1Name=base_hair_01_rt
  bone2Name=head
  swing1LimitDegrees=45
  swing1Motion=Locked
  swing2LimitDegrees=45
  swing2Motion=Locked
  twistLimitDegrees=45
  twistMotion=Locked
 }
 {bone1Name=dyn_hair_01_rt, bone2Name=base_hair_01_rt, swing1LimitDegrees=10, swing2LimitDegrees=25, twistLimitDegrees=45, twistMotion=Locked}
 {bone1Name=dyn_hair_02_rt, bone2Name=dyn_hair_01_rt, swing1LimitDegrees=10, swing2LimitDegrees=25, twistLimitDegrees=45, twistMotion=Locked}
 {bone1Name=dyn_hair_03_rt, bone2Name=dyn_hair_02_rt, swing1LimitDegrees=12, swing2LimitDegrees=25, twistLimitDegrees=45, twistMotion=Locked}
 {
  bone1Name=base_hair_01_back_lf
  bone2Name=head
  swing1LimitDegrees=45
  swing1Motion=Free
  swing2LimitDegrees=45
  swing2Motion=Free
  twistLimitDegrees=45
  twistMotion=Free
 }
 {
  bone1Name=dyn_hair_01_back_lf
  bone2Name=base_hair_01_back_lf
  swing1LimitDegrees=20
  swing2LimitDegrees=20
  twistLimitDegrees=45
  twistMotion=Locked
 }
 {
  bone1Name=dyn_hair_02_back_lf
  bone2Name=dyn_hair_01_back_lf
  swing1LimitDegrees=20
  swing2LimitDegrees=20
  twistLimitDegrees=45
  twistMotion=Locked
 }
 {
  bone1Name=dyn_hair_03_back_lf
  bone2Name=dyn_hair_02_back_lf
  swing1LimitDegrees=20
  swing2LimitDegrees=20
  twistLimitDegrees=45
  twistMotion=Locked
 }
 {
  bone1Name=base_hair_01_back_rt
  bone2Name=head
  swing1LimitDegrees=45
  swing1Motion=Free
  swing2LimitDegrees=45
  swing2Motion=Free
  twistLimitDegrees=45
  twistMotion=Free
 }
 {
  bone1Name=dyn_hair_01_back_rt
  bone2Name=base_hair_01_back_rt
  swing1LimitDegrees=20
  swing2LimitDegrees=20
  twistLimitDegrees=45
  twistMotion=Locked
 }
 {
  bone1Name=dyn_hair_02_back_rt
  bone2Name=dyn_hair_01_back_rt
  swing1LimitDegrees=20
  swing2LimitDegrees=20
  twistLimitDegrees=45
  twistMotion=Locked
 }
 {
  bone1Name=dyn_hair_03_back_rt
  bone2Name=dyn_hair_02_back_rt
  swing1LimitDegrees=20
  swing2LimitDegrees=20
  twistLimitDegrees=45
  twistMotion=Locked
 }
 {
  bone1Name=dyn_hair_02_back_lf
  bone2Name=dyn_hair_02_back
  swing1LimitDegrees=45
  swing1Motion=Free
  swing2LimitDegrees=45
  swing2Motion=Free
  twistLimitDegrees=45
  twistMotion=Free
 }
 {
  bone1Name=dyn_hair_02_back_rt
  bone2Name=dyn_hair_02_back
  swing1LimitDegrees=45
  swing1Motion=Free
  swing2LimitDegrees=45
  swing2Motion=Free
  twistLimitDegrees=45
  twistMotion=Free
 }
 {
  bone1Name=dyn_hair_03_back_lf
  bone2Name=dyn_hair_03_lf
  swing1LimitDegrees=45
  swing1Motion=Free
  swing2LimitDegrees=45
  swing2Motion=Free
  twistLimitDegrees=45
  twistMotion=Free
 }
 {
  bone1Name=dyn_hair_03_back_rt
  bone2Name=dyn_hair_03_rt
  swing1LimitDegrees=45
  swing1Motion=Free
  swing2LimitDegrees=45
  swing2Motion=Free
  twistLimitDegrees=45
  twistMotion=Free
 }
 {bone1Name=dyn_hair_top_L_01, bone2Name=head, swing1LimitDegrees=5, swing2LimitDegrees=10, twistLimitDegrees=45, twistMotion=Locked}
 {
  bone1Name=dyn_hair_top_L_02
  bone2Name=dyn_hair_top_L_01
  swing1LimitDegrees=10
  swing2LimitDegrees=10
  twistLimitDegrees=45
  twistMotion=Locked
 }
 {
  bone1Name=dyn_hair_top_L_03
  bone2Name=dyn_hair_top_L_02
  swing1LimitDegrees=15
  swing2LimitDegrees=10
  twistLimitDegrees=45
  twistMotion=Locked
 }
 {bone1Name=dyn_hair_top_R_01, bone2Name=head, swing1LimitDegrees=5, swing2LimitDegrees=10, twistLimitDegrees=45, twistMotion=Locked}
 {
  bone1Name=dyn_hair_top_R_02
  bone2Name=dyn_hair_top_R_01
  swing1LimitDegrees=15
  swing2LimitDegrees=10
  twistLimitDegrees=45
  twistMotion=Locked
 }
 {
  bone1Name=dyn_hair_top_R_03
  bone2Name=dyn_hair_top_R_02
  swing1LimitDegrees=15
  swing2LimitDegrees=10
  twistLimitDegrees=45
  twistMotion=Locked
 }
]
