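# Physics asset for the face accessory part of the male M_MED_AlpacaLean_Gel player cosmetic, defining bodies for secondary motion on its dynamic hair bones.
# Four three-bone capsule chains (left, right, top-left, top-right) are each constrained in sequence back to a head capsule.
# The left and right chains have free swing and twist, while the top chains lock twist and use narrow swing limits for stiffer movement.
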PhysicsAsset /BRCosmetics/Characters/Player/Male/Medium/Bodies/M_MED_AlpacaLean_Gel/Meshes/Parts/M_MED_AlpacaLean_Gel_FaceAcc_Physics1
bodies {
 dyn_hair_01_lf [{shapeType=Capsule, length=6, radius=3, shapeName=dyn_hair_01_lf_capsule, center=<0 2.1734 -1.2246>, rotation=<rot 0 0 -11.0613>}]
 dyn_hair_01_rt [{shapeType=Capsule, length=2.5, radius=2.5, shapeName=dyn_hair_01_rt_capsule, center=<0 -2.4038 -0.2583>}]
 dyn_hair_02_lf [{shapeType=Capsule, length=6, radius=3, shapeName=dyn_hair_02_lf_capsule, center=<0 1.8338 -1.1846>, rotation=<rot 0 0 -11.0613>}]
 dyn_hair_02_rt [{shapeType=Capsule, length=2.5, radius=2.5, shapeName=dyn_hair_02_rt_capsule, center=<0 -2.1694 0>}]
 dyn_hair_03_lf [{shapeType=Capsule, length=6, radius=3, shapeName=dyn_hair_03_lf_capsule, center=<0 2.1698 -1.1846>, rotation=<rot 0 0 -11.0613>}]
 dyn_hair_03_rt [{shapeType=Capsule, length=2.5, radius=2.5, shapeName=dyn_hair_03_rt_capsule, center=<0 -1.8316 0>}]
 dyn_hair_top_L_01 [{
   shapeType=Capsule
   length=2.8033
   radius=1
   shapeName=dyn_hair_top_L_01_capsule
   center=<4.161 -0.1116 0.1526>
   rotation=<rot 0 -89.3417 -81.9822>
  }]
 dyn_hair_top_L_02 [{
   shapeType=Capsule
   length=3.473
   radius=1
   shapeName=dyn_hair_top_L_02_capsule
   center=<3.4767 -0.0335 -0.0423>
   rotation=<rot 0 -88.9012 -88.1346>
  }]
 dyn_hair_top_L_03 [{
   shapeType=Capsule
   length=2.2366
   radius=1
   shapeName=dyn_hair_top_L_03_capsule
   center=<3.1568 0.0253 -1.1435>
   rotation=<rot 0 -88.6085 -52.0494>
  }]
 dyn_hair_top_R_01 [{
   shapeType=Capsule
   length=1.6129
   radius=1
   shapeName=dyn_hair_top_R_01_capsule
   center=<2.7176 0.368 -0.2745>
   rotation=<rot 0 83.2388 -88.95>
  }]
 dyn_hair_top_R_02 [{
   shapeType=Capsule
   length=3.7746
   radius=1
   shapeName=dyn_hair_top_R_02_capsule
   center=<3.0928 0.213 0.0444>
   rotation=<rot 0 94.3657 -89.9758>
  }]
 dyn_hair_top_R_03 [{
   shapeType=Capsule
   length=1.9025
   radius=1
   shapeName=dyn_hair_top_R_03_capsule
   center=<2.7004 0.3738 -0.0062>
   rotation=<rot 0 103.7471 -81.6998>
  }]
 head [{shapeType=Capsule, length=10, radius=5, center=<3.6306 0.7416 0.2467>, rotation=<rot 90 0 0>}]
}
constraints [
 {
  bone1Name=dyn_hair_03_lf
  bone2Name=dyn_hair_02_lf
  swing1LimitDegrees=45
  swing1Motion=Free
  swing2LimitDegrees=45
  swing2Motion=Free
  twistLimitDegrees=45
  twistMotion=Free
 }
 {
  bone1Name=dyn_hair_01_lf
  bone2Name=head
  swing1LimitDegrees=45
  swing1Motion=Free
  swing2LimitDegrees=45
  swing2Motion=Free
  twistLimitDegrees=45
  twistMotion=Free
 }
 {
  bone1Name=dyn_hair_02_lf
  bone2Name=dyn_hair_01_lf
  swing1LimitDegrees=45
  swing1Motion=Free
  swing2LimitDegrees=45
  swing2Motion=Free
  twistLimitDegrees=45
  twistMotion=Free
 }
 {bone1Name=dyn_hair_top_L_01, bone2Name=head, swing1LimitDegrees=15, swing2LimitDegrees=5, twistLimitDegrees=45, twistMotion=Locked}
 {
  bone1Name=dyn_hair_top_L_02
  bone2Name=dyn_hair_top_L_01
  swing1LimitDegrees=15
  swing2LimitDegrees=7
  twistLimitDegrees=45
  twistMotion=Locked
 }
 {
  bone1Name=dyn_hair_top_L_03
  bone2Name=dyn_hair_top_L_02
  swing1LimitDegrees=15
  swing2LimitDegrees=10
  twistLimitDegrees=45
  twistMotion=Locked
 }
 {bone1Name=dyn_hair_top_R_01, bone2Name=head, swing1LimitDegrees=15, swing2LimitDegrees=10, twistLimitDegrees=45, twistMotion=Locked}
 {
  bone1Name=dyn_hair_top_R_02
  bone2Name=dyn_hair_top_R_01
  swing1LimitDegrees=15
  swing2LimitDegrees=10
  twistLimitDegrees=45
  twistMotion=Locked
 }
 {
  bone1Name=dyn_hair_top_R_03
  bone2Name=dyn_hair_top_R_02
  swing1LimitDegrees=15
  swing2LimitDegrees=10
  twistLimitDegrees=45
  twistMotion=Locked
 }
 {
  bone1Name=dyn_hair_01_rt
  bone2Name=head
  swing1LimitDegrees=45
  swing1Motion=Free
  swing2LimitDegrees=45
  swing2Motion=Free
  twistLimitDegrees=45
  twistMotion=Free
 }
 {
  bone1Name=dyn_hair_02_rt
  bone2Name=dyn_hair_01_rt
  swing1LimitDegrees=45
  swing1Motion=Free
  swing2LimitDegrees=45
  swing2Motion=Free
  twistLimitDegrees=45
  twistMotion=Free
 }
 {
  bone1Name=dyn_hair_03_rt
  bone2Name=dyn_hair_02_rt
  swing1LimitDegrees=45
  swing1Motion=Free
  swing2LimitDegrees=45
  swing2Motion=Free
  twistLimitDegrees=45
  twistMotion=Free
 }
]
